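; B_Nitro_TestApplyVolume is a test volume Blueprint that applies a nitro gameplay effect to actors entering it.
; When something overlaps its Cylinder component, it gets the overlapping actor's Fort ability system component.
; On the server only, if that component is valid, it applies the ApplyNitroGE effect class to it at level 0.
; The construction script is empty.
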(event OnComponentBeginOverlap(Cylinder)
  (params (OverlappedComponent :type Primitive%20Component%20Object%20Reference) (OtherActor :type Actor%20Object%20Reference) (OtherComp :type Primitive%20Component%20Object%20Reference) (OtherBodyIndex :type Integer) (bFromSweep :type Boolean) (SweepResult :type Hit%20Result%20Structure%20%28by%20ref%29))
  (bind _returnvalue (Ability|GetFortAbilitySystemComponent OtherActor))
  (if (Networking|IsServer)
    (Utilities|IsValid _returnvalue
      (:Is%20Valid
        (GameplayEffects|ApplyGameplayEffectToSelf :self _returnvalue :GameplayEffectClass (Variables|Default|GetApplyNitroGE) :Level 0.0)))))

(fn ConstructionScript)
